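; PanZoom2D camera actor that mirrors a shot camera and pans and zooms its image in 2D through filmback changes.
; SetData copies shot cine camera settings and configures a SceneCapture2D, deriving focal length from FOV and sensor width for non-cine cameras.
; Tick resyncs differing focus, lens, crop, aperture and focal length, leaving filmback independent.
; updateFilmback scales the initial sensor size for zoom and sets clamped, zoom-dependent sensor offsets for pan.
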(event EventBeginPlay)

(event Collision|EventActorBeginOverlap
  (params (OtherActor :type Actor%20Object%20Reference)))

(event EventTick
  (params (DeltaSeconds :type Float%20%28single-precision%29))
  (CallFunction|MatchShotCam))

(fn MatchShotCam
  (bind _shotcinecamcomponent (Variables|Default|GetShotCineCamComponent))
  (bind _shotcinecamcomponent_2 (Variables|Default|GetShotCineCamComponent))
  (bind (_lenssettings_minfocallength _lenssettings_maxfocallength _lenssettings_minfstop _lenssettings_maxfstop _lenssettings_minimumfocusdistance _lenssettings_squeezefactor _lenssettings_diaphragmbladecount) (Utilities|Struct|BreakCameraLensSettings (Variables|/Script/CinematicCamera.CineCameraComponent|GetLensSettings_MinFocalLength _shotcinecamcomponent_2)))
  (bind _2dpancameracomponent (Variables|Default|Get2dPanCameraComponent))
  (bind (_focussettings_focusmethod _focussettings_manualfocusdistance _focussettings_trackingfocussettings _focussettings_bsmoothfocuschanges _focussettings_focussmoothinginterpspeed _focussettings_focusoffset) (Utilities|Struct|BreakCameraFocusSettings (Variables|/Script/CinematicCamera.CineCameraComponent|GetFocusSettings_FocusMethod (Variables|Default|Get2dPanCameraComponent))))
  (bind (_lenssettings_minfocallength_2 _lenssettings_maxfocallength_2 _lenssettings_minfstop_2 _lenssettings_maxfstop_2 _lenssettings_minimumfocusdistance_2 _lenssettings_squeezefactor_2 _lenssettings_diaphragmbladecount_2) (Utilities|Struct|BreakCameraLensSettings (Variables|/Script/CinematicCamera.CineCameraComponent|GetLensSettings_MinFocalLength (Variables|Default|Get2dPanCameraComponent))))
  (bind _cropsettings_aspectratio (Utilities|Struct|BreakPlateCropSettings (Variables|/Script/CinematicCamera.CineCameraComponent|GetCropSettings_AspectRatio _shotcinecamcomponent)))
  (bind _cropsettings_aspectratio_2 (Utilities|Struct|BreakPlateCropSettings (Variables|/Script/CinematicCamera.CineCameraComponent|GetCropSettings_AspectRatio _2dpancameracomponent)))
  (bind _2dpancameracomponent_2 (Variables|Default|Get2dPanCameraComponent))
  (bind _shotcinecamcomponent_3 (Variables|Default|GetShotCineCamComponent))
  (bind _shotcinecamcomponent_4 (Variables|Default|GetShotCineCamComponent))
  (bind _currentfocallength (Variables|/Script/CinematicCamera.CineCameraComponent|GetCurrentFocalLength _shotcinecamcomponent_3))
  (bind (_focussettings_focusmethod_2 _focussettings_manualfocusdistance_2 _focussettings_trackingfocussettings_2 _focussettings_bsmoothfocuschanges_2 _focussettings_focussmoothinginterpspeed_2 _focussettings_focusoffset_2) (Utilities|Struct|BreakCameraFocusSettings (Variables|/Script/CinematicCamera.CineCameraComponent|GetFocusSettings_FocusMethod (Variables|Default|GetShotCineCamComponent))))
  (if (not (Math|Boolean|ANDBoolean (== _focussettings_manualfocusdistance_2 _focussettings_manualfocusdistance) (== _focussettings_focussmoothinginterpspeed_2 _focussettings_focussmoothinginterpspeed) (== _focussettings_focusoffset_2 _focussettings_focusoffset)))
    (Variables|/Script/CinematicCamera.CineCameraComponent|SetFocusSettings (Variables|/Script/CinematicCamera.CineCameraComponent|GetFocusSettings _shotcinecamcomponent_2) (Variables|Default|Get2dPanCameraComponent)))
  (if (not (Math|Boolean|ANDBoolean (== _lenssettings_minfocallength _lenssettings_minfocallength_2) (== _lenssettings_maxfocallength _lenssettings_maxfocallength_2) (== _lenssettings_minfstop _lenssettings_minfstop_2) (== _lenssettings_maxfstop _lenssettings_maxfstop_2) (== _lenssettings_minimumfocusdistance _lenssettings_minimumfocusdistance_2) (== _lenssettings_squeezefactor _lenssettings_squeezefactor_2) (== _lenssettings_diaphragmbladecount _lenssettings_diaphragmbladecount_2)))
    (Variables|/Script/CinematicCamera.CineCameraComponent|SetLensSettings (Variables|/Script/CinematicCamera.CineCameraComponent|GetLensSettings _shotcinecamcomponent) _2dpancameracomponent))
  (if (not (== _cropsettings_aspectratio_2 _cropsettings_aspectratio))
    (Variables|/Script/CinematicCamera.CineCameraComponent|SetCropSettings (Variables|/Script/CinematicCamera.CineCameraComponent|GetCropSettings _shotcinecamcomponent_4) _2dpancameracomponent_2))
  (if (not (== (Variables|/Script/CinematicCamera.CineCameraComponent|GetCurrentAperture _2dpancameracomponent_2) (Variables|/Script/CinematicCamera.CineCameraComponent|GetCurrentAperture _shotcinecamcomponent_4)))
    (Variables|/Script/CinematicCamera.CineCameraComponent|SetCurrentAperture (Variables|/Script/CinematicCamera.CineCameraComponent|GetCurrentAperture _shotcinecamcomponent_3) _2dpancameracomponent_2))
  (if (not (== _currentfocallength (Variables|/Script/CinematicCamera.CineCameraComponent|GetCurrentFocalLength (Variables|Default|Get2dPanCameraComponent))))
    (Variables|/Script/CinematicCamera.CineCameraComponent|SetCurrentFocalLength _currentfocallength (Variables|Default|Get2dPanCameraComponent))))

(fn setfocal
  (params (Target :type Camera%20Component%20Object%20Reference))
  (Variables|/Script/CinematicCamera.CineCameraComponent|SetCurrentFocalLength (/ (/ (Utilities|Struct|BreakPostProcessSettings (Variables|/Script/Engine.CameraComponent|GetPostProcessSettings Target)) (Math|Trig|Tan(Degrees) (/ (Variables|/Script/Engine.CameraComponent|GetFieldOfView Target) 2.0))) 2.0) (Variables|Default|Get2dPanCameraComponent)))

(fn SetData
  (params (Camera :type Camera%20Component%20Object%20Reference))
  (bind _camera (Variables|GetCamera))
  (bind _postprocesssettings (Variables|/Script/Engine.CameraComponent|GetPostProcessSettings _camera))
  (bind _depthoffieldsensorwidth (Utilities|Struct|BreakPostProcessSettings _postprocesssettings))
  (bind _camera_2 (Variables|GetCamera))
  (bind _postprocesssettings_2 (Variables|/Script/Engine.CameraComponent|GetPostProcessSettings _camera_2))
  (bind _depthoffieldsensorwidth_2 (Utilities|Struct|BreakPostProcessSettings _postprocesssettings_2))
  (bind _camera_3 (Variables|GetCamera))
  (bind _aspectratio (Variables|/Script/Engine.CameraComponent|GetAspectRatio _camera_3))
  (bind _returnvalue (/ _depthoffieldsensorwidth_2 _aspectratio))
  (bind _shotcinecamcomponent (Variables|Default|GetShotCineCamComponent))
  (bind _filmback (Variables|/Script/CinematicCamera.CineCameraComponent|GetFilmback _shotcinecamcomponent))
  (bind _2dpancameracomponent (Variables|Default|Get2dPanCameraComponent))
  (bind _camera_4 (Variables|GetCamera))
  (bind _scenecapturecomponent2d (Variables|Default|GetSceneCaptureComponent2D))
  (bind _filmback_sensorwidth (Utilities|Struct|BreakCameraFilmbackSettings (Variables|/Script/CinematicCamera.CineCameraComponent|GetFilmback_SensorWidth (Variables|Default|Get2dPanCameraComponent))))
  (bind (_lenssettings_minfocallength _lenssettings_maxfocallength _lenssettings_minfstop _lenssettings_maxfstop _lenssettings_minimumfocusdistance _lenssettings_squeezefactor) (Utilities|Struct|BreakCameraLensSettings (Variables|/Script/CinematicCamera.CineCameraComponent|GetLensSettings_MinFocalLength (Variables|Default|Get2dPanCameraComponent))))
  (bind _scenecapturecomponent2d_2 (Variables|Default|GetSceneCaptureComponent2D))
  (bind _2dpancameracomponent_2 (Variables|Default|Get2dPanCameraComponent))
  (bind _2dpancameracomponent_3 (Variables|Default|Get2dPanCameraComponent))
  (bind _2dpancameracomponent_4 (Variables|Default|Get2dPanCameraComponent))
  (bind _ascine_camera_component (Utilities|Casting|CastToCineCameraComponent Camera)
    (:then
      (Variables|Default|SetShotCineCamComponent _ascine_camera_component)
      (Utilities|FlowControl|Sequence
        (:then_0
          (Variables|/Script/CinematicCamera.CineCameraComponent|SetFilmback _filmback _2dpancameracomponent)
          (Variables|Default|SetInitialFilmBack _filmback))
        (:then_1
          (Variables|/Script/CinematicCamera.CineCameraComponent|SetCropSettings (Variables|/Script/CinematicCamera.CineCameraComponent|GetCropSettings _shotcinecamcomponent) _2dpancameracomponent))
        (:then_2
          (Variables|/Script/CinematicCamera.CineCameraComponent|SetFocusSettings (Variables|/Script/CinematicCamera.CineCameraComponent|GetFocusSettings _shotcinecamcomponent) _2dpancameracomponent))
        (:then_3
          (Variables|/Script/CinematicCamera.CineCameraComponent|SetLensSettings (Variables|/Script/CinematicCamera.CineCameraComponent|GetLensSettings _shotcinecamcomponent) _2dpancameracomponent))
        (:then_4
          (Variables|/Script/CinematicCamera.CineCameraComponent|SetCurrentAperture (Variables|/Script/CinematicCamera.CineCameraComponent|GetCurrentAperture _shotcinecamcomponent) _2dpancameracomponent))
        (:then_5
          (Variables|/Script/CinematicCamera.CineCameraComponent|SetCurrentFocalLength (Variables|/Script/CinematicCamera.CineCameraComponent|GetCurrentFocalLength _shotcinecamcomponent) _2dpancameracomponent)
          (Variables|/Script/Engine.SceneCaptureComponent2D|SetOverscan (* (* _lenssettings_squeezefactor (Variables|/Script/Engine.CameraComponent|GetOverscan (Variables|Default|Get2dPanCameraComponent))) _filmback_sensorwidth) _scenecapturecomponent2d_2)
          (Variables|/Script/Engine.SceneCaptureComponent2D|SetFOVAngle (Variables|/Script/Engine.CameraComponent|GetFieldOfView _2dpancameracomponent_2) _scenecapturecomponent2d_2)
          (Variables|/Script/Engine.SceneCaptureComponent2D|SetPostProcessSettings (Variables|/Script/Engine.CameraComponent|GetPostProcessSettings _2dpancameracomponent_2) _scenecapturecomponent2d_2))))
    (:CastFailed
      (Variables|/Script/Engine.SceneCaptureComponent2D|SetOverscan (Variables|/Script/Engine.CameraComponent|GetOverscan _camera_4) _scenecapturecomponent2d)
      (Variables|/Script/Engine.SceneCaptureComponent2D|SetFOVAngle (Variables|/Script/Engine.CameraComponent|GetFieldOfView _camera_4) _scenecapturecomponent2d)
      (Variables|/Script/Engine.SceneCaptureComponent2D|SetPostProcessSettings (Variables|/Script/Engine.CameraComponent|GetPostProcessSettings _camera_4) _scenecapturecomponent2d)
      (Variables|/Script/CinematicCamera.CineCameraComponent|SetCurrentFocalLength (/ (/ _depthoffieldsensorwidth_2 (Math|Trig|Tan(Degrees) (/ (Variables|/Script/Engine.CameraComponent|GetFieldOfView _camera_2) 2.0))) 2.0) _2dpancameracomponent_3)
      (Variables|/Script/CinematicCamera.CineCameraComponent|SetFilmback_SensorWidth (Utilities|Struct|MakeCameraFilmbackSettings _depthoffieldsensorwidth _returnvalue) _2dpancameracomponent_3)
      (Variables|SetInitialFilmBack_SensorWidth (Utilities|Struct|MakeCameraFilmbackSettings _depthoffieldsensorwidth _returnvalue))
      (Variables|/Script/CinematicCamera.CineCameraComponent|SetCurrentAperture (Utilities|Struct|BreakPostProcessSettings (Variables|/Script/Engine.CameraComponent|GetPostProcessSettings (Variables|GetCamera))) _2dpancameracomponent_4)
      (Utilities|Struct|SetmembersinCameraFocusSettings (Variables|/Script/CinematicCamera.CineCameraComponent|GetFocusSettings _2dpancameracomponent_4) "DoNotOverride"))))

(fn updateFilmback
  (params (Scale :type Float%20%28double-precision%29) (PositionDifference :type Vector%202D%20Structure) (InitialFilmBack :type Camera%20Filmback%20Settings%20Structure))
  (bind (_x _y) (Math|Vector2D|BreakVector2D PositionDifference))
  (bind (_initialfilmback_sensorwidth _initialfilmback_sensorheight) (Utilities|Struct|BreakCameraFilmbackSettings (Variables|GetInitialFilmBack_SensorWidth)))
  (Variables|/Script/CinematicCamera.CineCameraComponent|SetFilmback_SensorWidth (Utilities|Struct|MakeCameraFilmbackSettings (* _initialfilmback_sensorwidth (Variables|GetScale)) (* _initialfilmback_sensorheight (Variables|GetScale)) (Math|Float|MapRangeClamped _x 0.0 1.0 (* (* (Variables|Default|GetHorOffsetMax) (Math|Float|MapRangeClamped (Variables|GetScale) 0.01 1.0 1.0 0.01)) -1.0) (* (Variables|Default|GetHorOffsetMax) (Math|Float|MapRangeClamped (Variables|GetScale) 0.01 1.0 1.0 0.01))) (* (Math|Float|MapRangeClamped _y 0.0 1.0 (* (* (Math|Float|MapRangeClamped (Variables|GetScale) 0.01 1.0 1.0 0.01) (Variables|Default|GetVertOffsetMax)) -1.0) (* (Math|Float|MapRangeClamped (Variables|GetScale) 0.01 1.0 1.0 0.01) (Variables|Default|GetVertOffsetMax))) -1.0)) (Variables|Default|Get2dPanCameraComponent)))

(fn ConstructionScript
  (bind _ascine_camera_component (Utilities|Casting|CastToCineCameraComponent (Variables|CameraActor|GetCameraComponent)))
  (Variables|Default|Set2dPanCameraComponent _ascine_camera_component))
