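# PanZoom2D_RenderTarget_Mat is a UI-domain material for the PanZoom2D level sequence tool, built around that tool's render target.
# Its only node is a Texture Sample that reads PanZoom2D_RenderTarget with a Linear Color sampler.
# The graph has no parameters or math, so it does no pan or zoom itself.
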Material /AnimInEdTools/EventsWipLevelSequenceTools/PanZoom2D/PanZoom2D_RenderTarget_Mat
graph {
 properties {materialDomain=MD_UI, usedShadingModels=Default Lit}
 nodes [TextureSample() {samplerType=SAMPLERTYPE_LinearColor, texture=&/PanZoom2D_RenderTarget}]
}
