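# Editor utility widget that adds and controls a 2D pan and zoom camera for the focused Level Sequence.
# It finds the source camera across subsequences, attaches a PanZoom2DCamera via attach track, and can pilot it.
# A draggable indicator and zoom slider update filmback size and offsets, keyable on float tracks.
# A timer follows sequence changes, tick refreshes the UI on frame changes, and a SaveGame stores source camera and filmback per sequence.
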EditorUtilityWidgetBlueprint /AnimInEdTools/EventsWipLevelSequenceTools/PanZoom2D/PanZoom2D_EWIGD
graphs [
 CalcPositionPercentage {nodes [
   n1 = CalcPositionPercentage!() -> (then, InCoordinates:Vector 2D Structure) {is=title K2Node_FunctionEntry}
   n2 = BreakVector2D(InVec=@n1)
   n3 = GetSize!(self:Canvas Panel Slot Object Reference=SlotAsCanvasSlot!(Widget:Widget Object Reference=Get 2dPanCanvas() -> (2dPanCanvas:Canvas Panel Object Reference)) -> (ReturnValue:Canvas Panel Slot Object Reference)) -> (ReturnValue_X:double, ReturnValue_Y:double)
   n4 = GetSize(self=SlotAsCanvasSlot(Widget=Get ZoomIndicator() -> (ZoomIndicator:Border Object Reference)))
   n5 = float - float(A=@n3, B=@n4)
   n6 = float - float(A=@n3.ReturnValue_Y, B=@n4.ReturnValue_Y)
   > Return Node!(execute, CoordinatePercentage:Vector 2D Structure=MakeVector2D(X=float / float(A=@n2, B=@n5), Y=float / float(A=@n2.Y, B=@n6)))
  ]}
 CheckForSequenceUpdate {nodes [
   CheckForSequenceUpdate() {is=title K2Node_FunctionEntry}
   > n1 = Get Current Local Time!(execute, TimeUnit:EMovieSceneTimeUnit Enum=DisplayRate) -> (
    then
    ReturnValue_Frame_FrameNumber_Value:int
    ReturnValue_Frame_SubFrame:float=~
    ReturnValue_Time:float=~
    ReturnValue_MarkedFrame:String
    ReturnValue_Timecode:Timecode Structure
    ReturnValue_PositionType:EMovieScenePositionType Enum
    ReturnValue_UpdateMethod:EUpdatePositionMethod Enum
    ReturnValue_bHasJumped:bool=~
   ) {is=title K2Node_CallFunction}
   > n2 = Branch(Condition="Equal (Integer)"(A=Get PreviousFrame() -> (PreviousFrame:int=0), B=@n1))
   > Set IsFrameDifferent(IsFrameDifferent:bool=false) -> (Output_Get:bool)
   Set PreviousFrame(execute=@n2.else, PreviousFrame:int=@n1) -> (Output_Get:int=0)
   > Set IsFrameDifferent(IsFrameDifferent=true)
   > n3 = Is Valid(InputObject=Get 2dPanCamera() -> (2dPanCamera:Pan Zoom 2DCamera Object Reference))
   SetVisibility(execute=@n3.Is Not Valid, self=Get ZoomIndicator(), InVisibility=Hidden)
   > n4 = FindExistingCamera!(execute, self) -> (then, 2dPanExists:bool)
   > Branch(Condition=@n4)
   > LoadSave()
   > Branch(Condition=IsChecked(self=Get AutoPilotCam() -> (AutoPilotCam:Editor Utility Check Box Object Reference)))
   > PilotLevelActor!(
    execute
    self:Level Editor Subsystem Object Reference=Get LevelEditorSubsystem()
    ActorToPilot:Actor Object Reference=Get 2dPanCamera()
    ViewportConfigKey:Name=None
   ) -> (then)
   > SetVisibility(self=Get ZoomIndicator(), InVisibility=Visible)
   n5 = GetPlaybackRange!(
    execute=@n3.Is Valid
    Sequence:Movie Scene Sequence Object Reference=Get FocusedSequence() -> (FocusedSequence:Level Sequence Object Reference)
   ) -> (then, ReturnValue_bHasStart:bool=~, ReturnValue_bHasEnd:bool=~, ReturnValue_InclusiveStart:int, ReturnValue_ExclusiveEnd:int)
   n6 = Get PreviousFrame() -> (PreviousFrame=0)
   > n7 = Branch(Condition=OR Boolean(A="integer > integer"(A=@n6, B=@n5.ReturnValue_ExclusiveEnd), B="integer < integer"(A=@n6, B=@n5.ReturnValue_InclusiveStart)))
   n8 = Get ZoomIndicator()
   > n9 = SetVisibility(self=@n8, InVisibility=Hidden)
   n10 = GetPilotLevelActor!(execute=[@n13.then, @n9.then], self:Level Editor Subsystem Object Reference=Get LevelEditorSubsystem(), ViewportConfigKey:Name=None) -> (then, ReturnValue:Actor Object Reference)
   n11 = "Equal (Object)"(A=@n10, B=Get 2dPanCamera())
   n12 = Get AutoPilotCam()
   > Branch(Condition=AND Boolean(A=NOT Boolean(A=@n11), B=IsChecked(self=@n12)))
   > PilotLevelActor(self=Get LevelEditorSubsystem(), ActorToPilot=Get 2dPanCamera(), ViewportConfigKey=None)
   n13 = SetVisibility(execute=@n7.else, self=@n8, InVisibility=Visible)
  ]}
 CheckForSourceCamera {nodes [
   CheckForSourceCamera!() -> (then, Sequence:Movie Scene Sequence Object Reference) {is=title K2Node_FunctionEntry}
   > n1 = GetBindings!(execute, Sequence:Movie Scene Sequence Object Reference=Get Sequence() -> (Sequence:Movie Scene Sequence Object Reference)) -> (then, ReturnValue:Array of Movie Scene Binding Proxy Structures)
   > n2 = For Each Loop with Break!(Exec:Exec, Array:Array of Movie Scene Binding Proxy Structures=@n1, Break:Exec=@n10.then) -> (LoopBody:Exec, Array Element:Movie Scene Binding Proxy Structure, Array Index:int, Completed:Exec)
   Return Node!(execute=@n2.Completed, ContainsSourceCam:bool=Get CamFound() -> (CamFound:bool))
   > n3 = GetObjectTemplate!(execute, InBinding:"Movie Scene Binding Proxy Structure (by ref)"=@n2) -> (then, ReturnValue:Object Reference)
   > n4 = Cast To Actor!(execute, Object=@n3) -> (then, CastFailed, AsActor:Actor Object Reference) {is=title K2Node_DynamicCast}
   n5 = GetDisplayName(Object=@n4)
   n6 = GetDisplayName(Object=Get SelectedCam() -> (SelectedCam:Actor Object Reference))
   n7 = "Equal Exactly (String)"(A=@n5, B=Get SourceCamName() -> (SourceCamName:String))
   n8 = "Equal (Object)"(A=Get Sequence(), B=Get SourceCamSeq() -> (SourceCamSeq:Level Sequence Object Reference))
   > Branch(Condition=OR Boolean(A="Equal Exactly (String)"(A=@n5, B=@n6), B=AND Boolean(A=@n7, B=@n8)))
   > Set CamFound(CamFound:bool=true) -> (Output_Get:bool)
   > Set BindingProxy(BindingProxy:Movie Scene Binding Proxy Structure=@n2) -> (Output_Get:Movie Scene Binding Proxy Structure)
   > n9 = Cast To LevelSequence!(execute, Object=Get Sequence()) -> (then, CastFailed, AsLevel Sequence:Level Sequence Object Reference) {is=title K2Node_DynamicCast}
   > Set SourceCamSeq(SourceCamSeq:Level Sequence Object Reference=@n9) -> (Output_Get:Level Sequence Object Reference)
   > n10 = Set SourceCamName(SourceCamName:String=@n5) -> (Output_Get:String)
  ]}
 CheckSeqChanged {nodes [
   CheckSeqChanged() {is=title K2Node_FunctionEntry}
   > n1 = GetFocusedLevelSequence!(execute) -> (then, ReturnValue:Level Sequence Object Reference)
   > n2 = Branch(Condition="Equal (Object)"(A=@n1, B=Get FocusedSequence()))
   n3 = Is Valid(exec=@n2.else, InputObject=@n1)
   Clear and Invalidate Timer by Handle!(execute=@n3.Is Not Valid, WorldContextObject:Object Reference, Handle=Get SeqCheckTimer() -> (SeqCheckTimer:Timer Handle Structure)) -> (then)
   Set FocusedSequence(execute=@n3.Is Valid, FocusedSequence:Level Sequence Object Reference=@n1) -> (Output_Get:Level Sequence Object Reference)
   > LoadSave()
  ]}
 CycleThroughSequences {nodes [
   n1 = CycleThroughSequences!() -> (then, Sequence:Movie Scene Sequence Object Reference) {is=title K2Node_FunctionEntry}
   > n2 = GetTracks!(execute, Sequence:Movie Scene Sequence Object Reference=@n1) -> (then, ReturnValue:Array of Movie Scene Track Object References)
   > n3 = For Each Loop(Array:Array of Movie Scene Track Object References=@n2) -> (Array Element:Movie Scene Track Object Reference)
   n4 = CheckForSourceCamera!(execute=@n3.Completed, self, Sequence:Movie Scene Sequence Object Reference=Get Sequence()) -> (then, ContainsSourceCam:bool) {is=title K2Node_CallFunction}
   > n5 = Branch(Condition=@n4)
   > Return Node!(execute)
   n6 = For Each Loop(
    Exec=@n5.else
    Array:Array of Movie Scene Sequence Object References=Get SubSequences() -> (SubSequences:Array of Movie Scene Sequence Object References)
   ) -> (Array Element:Movie Scene Sequence Object Reference)
   > CycleThroughSequences!(execute, self, Sequence:Movie Scene Sequence Object Reference=@n6) -> (then) {is=title K2Node_CallFunction}
   n7 = Cast To MovieSceneSubTrack!(execute=@n3.LoopBody, Object=@n3) -> (then, CastFailed, AsMovie Scene Sub Track:Movie Scene Sub Track Object Reference) {is=title K2Node_DynamicCast}
   > n8 = GetSections!(execute, Track:Movie Scene Track Object Reference=@n7) -> (then, ReturnValue:Array of Movie Scene Section Object References)
   > n9 = For Each Loop(Array:Array of Movie Scene Section Object References=@n8) -> (Array Element:Movie Scene Section Object Reference)
   > n10 = Cast To MovieSceneSubSection!(execute, Object=@n9) -> (then, CastFailed, AsMovie Scene Sub Section:Movie Scene Sub Section Object Reference) {is=title K2Node_DynamicCast}
   > Add Unique(
    TargetArray:Array of Movie Scene Sequence Object References=Get SubSequences()
    NewItem:"Movie Scene Sequence Object Reference (by ref)"=GetSequence!(self:Movie Scene Sub Section Object Reference=@n10) -> (ReturnValue:Movie Scene Sequence Object Reference)
   )
  ]}
 EventGraph {nodes [
   "On Clicked (AddToolBtn)"() {is=title K2Node_ComponentBoundEvent}
   > n1 = GetCurrentLevelSequence!(execute) -> (then, ReturnValue:Level Sequence Object Reference)
   > Set RootSequence(RootSequence:Level Sequence Object Reference=@n1) -> (Output_Get:Level Sequence Object Reference)
   > n2 = GetFocusedLevelSequence()
   > n3 = Set FocusedSequence(FocusedSequence=@n2)
   > SaveLoadedAsset(AssetToSave=@n3, bOnlyIfIsDirty=true)
   > n4 = FindExistingCamera()
   > n5 = Branch(Condition=@n4)
   > SetData()
   Spawn2dPanCamera(execute=@n5.else)
   n6 = "On Value Changed (ZoomSlider)"!() -> (OutputDelegate, then, Value:float) {is=title K2Node_ComponentBoundEvent}
   > n7 = Branch(Condition=Get SliderUpdated() -> (SliderUpdated:bool))
   > Set SliderUpdated(SliderUpdated:bool=false) -> (Output_Get:bool)
   ScaleIndicator!(execute=@n7.else, self, Value:double=@n6.Value) -> (then)
   n8 = Get 2dPanCamera()
   > n9 = Is Valid(InputObject=@n8)
   n10 = FindExistingCamera(execute=@n9.Is Not Valid)
   updateFilmback!(
    execute=[@n10.then, @n9.Is Valid]
    self:Pan Zoom 2DCamera Object Reference=@n8
    Scale:double=Get ScaleValue() -> (ScaleValue:double)
    PositionDifference:Vector 2D Structure=Get PositionPercentage() -> (PositionPercentage:Vector 2D Structure)
    InitialFilmBack:Camera Filmback Settings Structure=Get DefaultFilmback() -> (DefaultFilmback:Camera Filmback Settings Structure)
   ) -> (then)
   "On Clicked (KeyPosBtn)"() {is=title K2Node_ComponentBoundEvent}
   > n11 = GetFocusedLevelSequence()
   > Set FocusedSequence(FocusedSequence=@n11)
   > KeyCamPosition()
   "On Clicked (ToggleCamBtn)"() {is=title K2Node_ComponentBoundEvent}
   > ToggleCam()
   "On Clicked (ClearSaveBtn)"() {is=title K2Node_ComponentBoundEvent}
   > n12 = Is Valid(InputObject=Get SaveGame() -> (SaveGame:SG Pan Zoom Object Reference))
   n13 = CreateSaveGameObject!(
    execute=@n12.Is Not Valid
    SaveGameClass:Save Game Class Reference=&/SG_PanZoom.SG_PanZoom_C
   ) -> (then, ReturnValue:SG Pan Zoom Object Reference)
   > n14 = Set SaveGame(SaveGame:SG Pan Zoom Object Reference=@n13) -> (Output_Get:SG Pan Zoom Object Reference)
   n15 = LoadGameFromSlot!(execute=[@n12.Is Valid, @n14.then], SlotName:String=PanZoom, UserIndex:int=0) -> (then, ReturnValue:Save Game Object Reference)
   > n16 = Cast To SG_PanZoom!(execute, Object=@n15) -> (then, CastFailed, AsSG Pan Zoom:SG Pan Zoom Object Reference) {is=title K2Node_DynamicCast}
   > n17 = Set SaveGame(SaveGame=@n16)
   > Clear!(
    execute
    TargetMap:Map of Level Sequence Object References to Pan Zoom Save Struct Structures=Get LvlSequences!(self:SG Pan Zoom Object Reference=@n17) -> (LvlSequences:Map of Level Sequence Object References to Pan Zoom Save Struct Structures)
   ) -> (then) {is=title K2Node_CallFunction}
   Event Construct()
   n18 = Get ZoomIndicator()
   > Set OriginalIndicatorSize(OriginalIndicatorSize:Vector 2D Structure=GetSize!(self=SlotAsCanvasSlot(Widget=@n18)) -> (ReturnValue:Vector 2D Structure)) -> (Output_Get:Vector 2D Structure)
   > n19 = GetCurrentLevelSequence()
   > Set RootSequence(RootSequence=@n19)
   > n20 = GetFocusedLevelSequence()
   > Set FocusedSequence(FocusedSequence=@n20)
   > LoadSave()
   > n21 = Set Timer by Function Name!(
    execute
    Object:Object Reference
    FunctionName:String=CheckSeqChanged
    Time:float=1
    bLooping:bool=true
    bMaxOncePerFrame:bool=false
    InitialStartDelay:float=0
    InitialStartDelayVariance:float=0
   ) -> (then, ReturnValue:Timer Handle Structure) {is=title K2Node_CallFunction}
   > Set SeqCheckTimer(SeqCheckTimer:Timer Handle Structure=@n21) -> (Output_Get:Timer Handle Structure)
   Event Tick()
   > n22 = Branch(Condition=Get IsDragging() -> (IsDragging:bool))
   n23 = Get 2dPanCamera()
   > n24 = Is Valid(InputObject=@n23)
   FindExistingCamera(execute=@n24.Is Not Valid)
   UpdateIndicatorPosition(execute=@n24.Is Valid)
   > updateFilmback(self=@n23, Scale=Get ScaleValue(), PositionDifference=Get PositionPercentage(), InitialFilmBack=Get DefaultFilmback())
   CheckForSequenceUpdate(execute=@n22.else) {is=title K2Node_CallFunction}
   > Branch(Condition=Get IsFrameDifferent() -> (IsFrameDifferent:bool))
   > UpdateUI()
   Event OnMouseLeave!() -> (OutputDelegate, then, MouseEvent:"Pointer Event Structure (by ref)") {is=title K2Node_Event}
   > Set IsDragging(IsDragging:bool=false) -> (Output_Get:bool)
  ]}
 FindExistingCamera {nodes [
   FindExistingCamera() {is=title K2Node_FunctionEntry}
   > n1 = GetAllActorsOfClass!(
    execute
    WorldContextObject:Object Reference
    ActorClass:Actor Class Reference=&/PanZoom2DCamera.PanZoom2DCamera_C
   ) -> (then, OutActors:Array of Pan Zoom 2DCamera Object References)
   > n2 = Set PanZooms(PanZooms:Array of Pan Zoom 2DCamera Object References=@n1) -> (Output_Get:Array of Pan Zoom 2DCamera Object References)
   > n3 = For Each Loop(Array:Array of Pan Zoom 2DCamera Object References=@n2) -> (Array Element:Pan Zoom 2DCamera Object Reference)
   n4 = Append(A=PanZoom_, B=GetDisplayName(Object=Get FocusedSequence()))
   n5 = Get PanZoomNames() -> (PanZoomNames:Array of Strings)
   Set 2dPanCamera(
    execute=@n3.Completed
    2dPanCamera:Pan Zoom 2DCamera Object Reference="Get (a copy)"(
     Array:Array of Pan Zoom 2DCamera Object References=Get PanZooms() -> (PanZooms:Array of Pan Zoom 2DCamera Object References)
     Dimension 1=Find Item!(TargetArray=@n5, ItemToFind:"String (by ref)"=@n4) -> (ReturnValue)
    ) -> (Output:Pan Zoom 2DCamera Object Reference)
   ) -> (Output_Get:Pan Zoom 2DCamera Object Reference)
   > Return Node!(execute, 2dPanExists:bool=IsValid(Object=Get 2dPanCamera()))
   Add(execute=@n3.LoopBody, TargetArray=Get PanZoomNames(), NewItem=GetDisplayName(Object=@n3)) {is=title K2Node_CallArrayFunction}
  ]}
 GetCamOffset {nodes [
   GetCamOffset() {is=title K2Node_FunctionEntry}
   n1 = MapRangeClamped(Value=Get ScaleValue(), InRangeA=0.01, InRangeB=1, OutRangeA=1, OutRangeB=0.01)
   n2 = Get 2dPanCamera()
   n3 = Get HorOffsetMax!(self:Pan Zoom 2DCamera Object Reference=@n2) -> (HorOffsetMax:double)
   n4 = Get!(self:Cine Camera Component Object Reference=Get 2dPanCameraComponent!(self:Pan Zoom 2DCamera Object Reference=@n2) -> (2dPanCameraComponent:Cine Camera Component Object Reference)) -> (
    Filmback_SensorWidth:float=~
    Filmback_SensorHeight:float=~
    Filmback_SensorHorizontalOffset:float=~
    Filmback_SensorVerticalOffset:float=~
    Filmback_SensorAspectRatio:float=~
   )
   n5 = Get VertOffsetMax!(self:Pan Zoom 2DCamera Object Reference=@n2) -> (VertOffsetMax:double)
   n6 = MakeVector2D(
    X=float / float(A=@n4.Filmback_SensorHorizontalOffset, B=float * float(A=@n3, B=@n1) {is=title K2Node_PromotableOperator})
    Y=float / float(A=float * float(A=@n4.Filmback_SensorVerticalOffset, B=-1), B=float * float(A=@n5, B=@n1))
   )
   n7 = Get ScaleValue()
   > n8 = Branch(Condition=AND Boolean(
     A="Equal (Vector2D)"!(A:Vector 2D Structure=@n6, B:Vector 2D Structure, ErrorTolerance:float) -> (ReturnValue:bool=~) {is=title K2Node_PromotableOperator}
     B="Equal (Float)"(A=@n7, B=1)
    ))
   > n9 = Set Position(Position:Vector 2D Structure=(X=-1,Y=-1)) -> (Output_Get:Vector 2D Structure)
   Return Node!(execute=[@n9.then, @n10.then], ReturnValue:Vector 2D Structure=Get Position() -> (Position:Vector 2D Structure))
   n10 = Set Position(execute=@n8.else, Position=@n6)
  ]}
 GetPanZoomCamRef {nodes [
   GetPanZoomCamRef() {is=title K2Node_FunctionEntry}
   > n1 = GetBindings(Sequence=Get FocusedSequence())
   > n2 = For Each Loop(Array:Array of Movie Scene Binding Proxy Structures=@n1) -> (Array Element:Movie Scene Binding Proxy Structure)
   > n3 = GetObjectTemplate(InBinding=@n2)
   > n4 = Cast To PanZoom2DCamera!(execute, Object=@n3) -> (then, CastFailed, AsPan Zoom 2DCamera:Pan Zoom 2DCamera Object Reference) {is=title K2Node_DynamicCast}
   > n5 = Set 2dPanCamera(2dPanCamera=@n4)
   > Return Node!(execute, AsPan Zoom 2DCamera:Pan Zoom 2DCamera Object Reference=@n5)
  ]}
 GetSelectedCamInfo {nodes [
   GetSelectedCamInfo() {is=title K2Node_FunctionEntry}
   > n1 = GetSelectedActors()
   n2 = Length(TargetArray=@n1)
   > n3 = Branch(Condition="Equal (Integer)"(A=@n2, B=1))
   > n4 = For Each Loop(Array:Array of Actor Object References=@n1) -> (Array Element:Actor Object Reference)
   n5 = CycleThroughSequences(execute=[@n12.Is Valid, @n4.Completed], Sequence=Get RootSequence() -> (RootSequence:Level Sequence Object Reference))
   n6 = IsValid(Object=Get SelectedCam())
   Return Node!(execute=[@n5.then, @n11.then], IsCamSelected:bool=@n6)
   n7 = Is Valid(
    exec=@n4.LoopBody
    InputObject=GetComponentByClass!(self:Actor Object Reference=@n4, ComponentClass:Actor Component Class Reference=/Script/Engine.CameraComponent) -> (ReturnValue:Camera Component Object Reference)
   )
   PrintString(execute=@n7.Is Not Valid, InString=Selected asset does not contain a camera)
   n8 = Get AddToolBtnTxt() -> (AddToolBtnTxt:Text Object Reference)
   > Set Text(Text="LOCTEXT(\"Select a camera\")", self=@n8)
   n9 = Get AddToolBtn() -> (AddToolBtn:Editor Utility Button Object Reference)
   > SetBackgroundColor!(execute, self:Button Object Reference=@n9, InBackgroundColor:Linear Color Structure=(R=0.1198,G=0.0148,B=0.0148,A=1)) -> (then)
   Set SelectedCam(execute=@n7.Is Valid, SelectedCam:Actor Object Reference=@n4) -> (Output_Get:Actor Object Reference)
   > Set Text(Text="LOCTEXT(\"Add 2d Pan / Reset\")", self=@n8)
   > SetBackgroundColor(self=@n9)
   n10 = Branch(execute=@n3.else, Condition="integer > integer"(A=@n2, B=1))
   > Set Text(Text="LOCTEXT(\"Select only 1 camera\")", self=Get AddToolBtnTxt())
   > n11 = SetBackgroundColor(self=Get AddToolBtn(), InBackgroundColor=(R=0.1198,G=0.0148,B=0.0148,A=1))
   n12 = Is Valid(exec=@n10.else, InputObject=Get SourceCamSeq())
   Set Text(execute=@n12.Is Not Valid, Text="LOCTEXT(\"Select a camera\")", self=Get AddToolBtnTxt())
   > SetBackgroundColor(self=Get AddToolBtn(), InBackgroundColor=(R=0.1198,G=0.0148,B=0.0148,A=1))
   > Return Node(IsCamSelected=@n6)
  ]}
 GetSourceCamComponentRef {nodes [
   GetSourceCamComponentRef() {is=title K2Node_FunctionEntry}
   > n1 = GetBindings(Sequence=Get SourceCamSeq())
   > n2 = For Each Loop(Array:Array of Movie Scene Binding Proxy Structures=@n1) -> (Array Element:Movie Scene Binding Proxy Structure)
   > n3 = GetObjectTemplate(InBinding=@n2)
   > n4 = Cast To CineCameraActor!(execute, Object=@n3) -> (then, CastFailed, AsCine Camera Actor:Cine Camera Actor Object Reference) {is=title K2Node_DynamicCast}
   n5 = "Equal Exactly (String)"(A=GetDisplayName(Object=@n3), B=Get SourceCamName())
   > Branch(Condition=@n5)
   > Set ShotCineCamRef(ShotCineCamRef:Cine Camera Component Object Reference=GetCineCameraComponent!(self:Cine Camera Actor Object Reference=@n4) -> (ReturnValue:Cine Camera Component Object Reference)) -> (Output_Get:Cine Camera Component Object Reference)
   > Set SelectedCam(SelectedCam=@n4)
   n6 = Cast To Actor(execute=@n4.CastFailed, Object=@n3)
   > Branch(Condition=@n5)
   > Set ShotCamRef(ShotCamRef:Camera Component Object Reference=GetComponentByClass(self=@n6, ComponentClass=&CameraComponent)) -> (Output_Get:Camera Component Object Reference)
   > Set SelectedCam(SelectedCam=@n6)
  ]}
 GetValues {nodes [
   GetValues() {is=title K2Node_FunctionEntry}
   n1 = Get FilmbackValues() -> (FilmbackValues:"Array of Float (double-precision)s")
   n2 = Get(self=Get 2dPanCameraComponent(self=Get 2dPanCamera())) -> (
    Filmback_SensorWidth=~
    Filmback_SensorHeight=~
    Filmback_SensorHorizontalOffset=~
    Filmback_SensorVerticalOffset=~
    Filmback_SensorAspectRatio=~
   )
   > Add(TargetArray:"Array of Float (double-precision)s"=@n1, NewItem:"Float (double-precision) (by ref)"=@n2)
   > Add(TargetArray=@n1, NewItem=@n2.Filmback_SensorHeight)
   > Add(TargetArray=@n1, NewItem=@n2.Filmback_SensorHorizontalOffset)
   > Add(TargetArray=@n1, NewItem=@n2.Filmback_SensorVerticalOffset)
   > Return Node!(execute, FilmbackValues:"Array of Float (double-precision)s"=Get FilmbackValues())
  ]}
 HookUpPanZoom {nodes [
   HookUpPanZoom() {is=title K2Node_FunctionEntry}
   n1 = Get SelectedCam()
   > n2 = Cast To CineCameraActor(Object=@n1)
   n3 = Get 2dPanCamera()
   > n4 = SetData!(execute, self:Pan Zoom 2DCamera Object Reference=@n3, Camera:Camera Component Object Reference=GetCineCameraComponent(self=@n2)) -> (then)
   n5 = AddTrack!(
    execute=[@n4.then, @n21.then]
    InBinding:"Movie Scene Binding Proxy Structure (by ref)"=Get PanZoomBinding() -> (PanZoomBinding:Movie Scene Binding Proxy Structure)
    TrackType:Movie Scene Track Class Reference=/Script/MovieSceneTracks.MovieScene3DAttachTrack
   ) -> (then, ReturnValue:Movie Scene Track Object Reference)
   > n6 = Cast To MovieScene3DAttachTrack!(execute, Object=@n5) -> (then, CastFailed, AsMovie Scene 3DAttach Track:Movie Scene 3DAttach Track Object Reference) {is=title K2Node_DynamicCast}
   > n7 = AddSection!(execute, Track:Movie Scene Track Object Reference=@n6) -> (then, ReturnValue:Movie Scene Section Object Reference)
   > n8 = Cast To MovieScene3DConstraintSection!(execute, Object=@n7) -> (then, CastFailed, AsMovie Scene 3DConstraint Section:Movie Scene 3DConstraint Section Object Reference) {is=title K2Node_DynamicCast}
   > Set MovieScene3dConstraintSection(MovieScene3dConstraintSection:Movie Scene 3DConstraint Section Object Reference=@n8) -> (Output_Get:Movie Scene 3DConstraint Section Object Reference)
   > n9 = Sequence!(execute) -> (then_0, then_1, then_2:Exec)
   > n10 = Branch(Condition="Equal (Object)"(A=Get FocusedSequence(), B=Get SourceCamSeq()))
   > n11 = GetBindingID!(
    execute
    Sequence:Movie Scene Sequence Object Reference=Get FocusedSequence()
    InBinding:"Movie Scene Binding Proxy Structure (by ref)"=Get BindingProxy() -> (BindingProxy:Movie Scene Binding Proxy Structure)
   ) -> (then, ReturnValue:Movie Scene Object Binding ID Structure)
   > Set Spawnable Bind ID(Spawnable Bind ID:Movie Scene Object Binding ID Structure=@n11) -> (Output_Get:Movie Scene Object Binding ID Structure)
   n12 = GetPortableBindingID!(
    execute=@n10.else
    RootSequence:Movie Scene Sequence Object Reference=Get RootSequence()
    DestinationSequence:Movie Scene Sequence Object Reference=Get FocusedSequence()
    InBinding:"Movie Scene Binding Proxy Structure (by ref)"=Get BindingProxy()
   ) -> (then, ReturnValue:Movie Scene Object Binding ID Structure)
   > Set Spawnable Bind ID(Spawnable Bind ID=@n12)
   SetConstraintBindingID!(
    execute=@n9.then_1
    self:Movie Scene 3DConstraint Section Object Reference=Get MovieScene3dConstraintSection() -> (MovieScene3dConstraintSection:Movie Scene 3DConstraint Section Object Reference)
    InConstraintBindingID:"Movie Scene Object Binding ID Structure (by ref)"=Get Spawnable Bind ID() -> (Spawnable Bind ID:Movie Scene Object Binding ID Structure)
   ) -> (then)
   n13 = Get FocusedSequence()
   > n14 = GetPlaybackStart!(execute, Sequence:Movie Scene Sequence Object Reference=@n13) -> (then, ReturnValue:int=0)
   n15 = Get MovieScene3dConstraintSection()
   > SetStartFrame!(execute, Section:Movie Scene Section Object Reference=@n15, StartFrame:int=@n14) -> (then)
   > n16 = GetPlaybackEnd!(execute, Sequence:Movie Scene Sequence Object Reference=@n13) -> (then, ReturnValue:int=0)
   > SetEndFrame!(execute, Section:Movie Scene Section Object Reference=@n15, EndFrame:int=@n16) -> (then)
   n17 = GetChildPossessables!(execute=@n9.then_2, InBinding:"Movie Scene Binding Proxy Structure (by ref)"=Get PanZoomBinding()) -> (then, ReturnValue:Array of Movie Scene Binding Proxy Structures)
   n18 = "Get (a ref)"(Array:Array of Movie Scene Binding Proxy Structures=@n17, Dimension 1=0) -> (Output:"Movie Scene Binding Proxy Structure (by ref)")
   > n19 = FindTracksByType!(
    execute
    InBinding:"Movie Scene Binding Proxy Structure (by ref)"=@n18
    TrackType:Movie Scene Track Class Reference=/Script/MovieSceneTracks.MovieSceneFloatTrack
   ) -> (then, ReturnValue:Array of Movie Scene Float Track Object References)
   > n20 = For Each Loop(Array:Array of Movie Scene Float Track Object References=@n19) -> (Array Element:Movie Scene Float Track Object Reference)
   SetData!(execute=@n20.Completed, self:Self Object Reference) -> (then)
   > PilotLevelActor(self=Get LevelEditorSubsystem(), ActorToPilot=Get 2dPanCamera(), ViewportConfigKey=None)
   > Set Visibility(Visibility=Visible, self=Get ZoomIndicator())
   RemoveTrack!(execute=@n20.LoopBody, InBinding:"Movie Scene Binding Proxy Structure (by ref)"=@n18, TrackToRemove:Movie Scene Track Object Reference=@n20) -> (then)
   n21 = SetData!(
    execute=@n2.CastFailed
    self:Pan Zoom 2DCamera Object Reference=@n3
    Camera:Camera Component Object Reference=GetComponentByClass(self=@n1, ComponentClass=&CameraComponent)
   ) -> (then)
  ]}
 Is2dPanCam {nodes [
   n1 = Is2dPanCam!() -> (then, InBinding:"Movie Scene Binding Proxy Structure (by ref)") {is=title K2Node_FunctionEntry}
   > n2 = GetPossessedObjectClass!(execute, InBinding:"Movie Scene Binding Proxy Structure (by ref)"=@n1) -> (then, ReturnValue:Object Class Reference)
   > Branch(Condition="Equal (Class)"(A=@n2, B=/Script/CinematicCamera.CineCameraComponent))
   > n3 = GetParent!(execute, InBinding:"Movie Scene Binding Proxy Structure (by ref)"=@n1) -> (then, ReturnValue:Movie Scene Binding Proxy Structure)
   > n4 = GetPossessedObjectClass(InBinding=@n3)
   n5 = "Equal (Class)"(A=@n4, B=&PanZoom2DCamera_C)
   > Branch(Condition=@n5)
   > Return Node!(execute, ReturnValue:bool=@n5)
  ]}
 KeyCamPosition {nodes [
   KeyCamPosition() {is=title K2Node_FunctionEntry}
   > n1 = GetBindings(Sequence=Get FocusedSequence())
   > n2 = For Each Loop with Break(Array=@n1, Break=@n3.then_0)
   > Branch(Condition=Is2dPanCam!(InBinding:"Movie Scene Binding Proxy Structure (by ref)"=@n2) -> (ReturnValue:bool) {is=title K2Node_CallFunction})
   > n3 = Sequence!(execute) -> (then_0, then_1)
   n4 = Set PanToolCamBinding(execute=@n3.then_1, PanToolCamBinding:Movie Scene Binding Proxy Structure=@n2) -> (Output_Get:Movie Scene Binding Proxy Structure)
   > n5 = GetTracks!(execute, InBinding:"Movie Scene Binding Proxy Structure (by ref)"=@n4) -> (then, ReturnValue)
   > n6 = For Each Loop(Array:Array of Movie Scene Track Object References=@n5) -> (Array Element:Movie Scene Track Object Reference)
   n7 = Get Tracks Data() -> (Tracks Data:Array of Filmback Struct Structures)
   n8 = For Each Loop!(Exec=@n6.Completed, Array:Array of Filmback Struct Structures=@n7) -> (
    LoopBody
    Array Element_TrackName_2_#1:Text
    Array Element_PropertyName_5_#2:Name
    Array Element_PropertyPath_8_#3:String
    Array Index
    Completed
   )
   > n9 = Branch(Condition=Find(TargetMap:Map of Strings to Booleans=Get HasTrack() -> (HasTrack:Map of Strings to Booleans), Key="To String (Text)"(InText=@n8)) -> (Value:bool))
   n10 = AddTrack(
    execute=@n9.else
    InBinding=Get PanToolCamBinding() -> (PanToolCamBinding:Movie Scene Binding Proxy Structure)
    TrackType=&MovieSceneFloatTrack
   )
   > n11 = Cast To MovieSceneFloatTrack!(execute, Object=@n10) -> (then, CastFailed, AsMovie Scene Float Track:Movie Scene Float Track Object Reference) {is=title K2Node_DynamicCast}
   > SetDisplayName!(execute, Track:Movie Scene Track Object Reference=@n11, InName:"Text (by ref)"=MakeLiteralText(Value=@n8)) -> (then)
   > n12 = AddSection(Track=@n11)
   > n13 = GetPlaybackRange(Sequence=Get FocusedSequence()) -> (ReturnValue_bHasStart=~, ReturnValue_bHasEnd=~)
   > SetRange!(
    execute
    Section:Movie Scene Section Object Reference=@n12
    StartFrame:int=@n13.ReturnValue_InclusiveStart
    EndFrame:int=int + int(A=@n13.ReturnValue_ExclusiveEnd)
   ) -> (then)
   > SetPropertyNameAndPath!(
    execute
    Track:Movie Scene Property Track Object Reference=@n11
    InPropertyName:"Name (by ref)"=@n8.Array Element_PropertyName_5_#2
    InPropertyPath:String=@n8.Array Element_PropertyPath_8_#3
   ) -> (then)
   > SetKey!(
    execute
    self
    Track:Movie Scene Track Object Reference=@n11
    NewValue:float="Get (a copy)"(
     Array:"Array of Float (double-precision)s"=GetValues!() -> (FilmbackValues:"Array of Float (double-precision)s") {is=title K2Node_CallFunction}
     Dimension 1=@n8.Array Index
    ) -> (Output:double)
   ) -> (then)
   n14 = For Each Loop(Exec=@n6.LoopBody, Array=@n7)
   > n15 = GetDisplayName!(execute:Exec, Track:Movie Scene Track Object Reference=@n6) -> (then:Exec, ReturnValue:Text)
   > Branch(Condition="Equal Exactly (Text)"!(A:"Text (by ref)"=@n15, B:"Text (by ref)"=@n14) -> (ReturnValue:bool))
   > Add!(
    execute
    TargetMap:Map of Strings to Booleans=Get HasTrack()
    Key:"String (by ref)"="To String (Text)"(InText=@n14)
    Value:"Boolean (by ref)"=true
   ) -> (then) {is=title K2Node_CallFunction}
   > SetKey(Track=@n6, NewValue="Get (a copy)"(Array=GetValues(), Dimension 1=@n14.Array Index))
  ]}
 LoadSave {nodes [
   LoadSave() {is=title K2Node_FunctionEntry}
   > n1 = Is Valid(InputObject=Get SaveGame())
   n2 = CreateSaveGameObject(execute=@n1.Is Not Valid, SaveGameClass=&SG_PanZoom_C)
   > n3 = Set SaveGame(SaveGame=@n2)
   n4 = LoadGameFromSlot(execute=[@n1.Is Valid, @n3.then], SlotName=PanZoom, UserIndex=0)
   > n5 = Cast To SG_PanZoom(Object=@n4)
   > Set SaveGame(SaveGame=@n5)
   n6 = Find!(
    TargetMap:Map of Level Sequence Object References to Pan Zoom Save Struct Structures=Get LvlSequences(self=Get SaveGame())
    Key:"Level Sequence Object Reference (by ref)"=Get FocusedSequence()
   ) -> (
    Value_SourceCamName_23_#4:String
    Value_SourceCamSeq_25_#5:Level Sequence Object Reference
    Value_CameraFilmBack_9_#6:Camera Filmback Settings Structure
    ReturnValue
   )
   > n7 = Branch(Condition=@n6.ReturnValue)
   > Set SourceCamName(SourceCamName=@n6)
   > Set SourceCamSeq(SourceCamSeq=@n6.Value_SourceCamSeq_25_#5)
   > GetSourceCamComponentRef() {is=title K2Node_CallFunction}
   > GetPanZoomCamRef!(execute, self) -> (then, AsPan Zoom 2DCamera:Pan Zoom 2DCamera Object Reference) {is=title K2Node_CallFunction}
   > Set DefaultFilmback(DefaultFilmback:Camera Filmback Settings Structure=@n6.Value_CameraFilmBack_9_#6) -> (Output_Get:Camera Filmback Settings Structure)
   n8 = Get 2dPanCamera()
   > Set Filmback!(
    execute
    Filmback:Camera Filmback Settings Structure=@n6.Value_CameraFilmBack_9_#6
    self:Pan Zoom 2DCamera Object Reference=@n8
   ) -> (then, Output_Get:Camera Filmback Settings Structure)
   > Set Filmback(
    Filmback=@n6.Value_CameraFilmBack_9_#6
    self:Cine Camera Component Object Reference=GetCineCameraComponent(self=@n8)
   )
   > SetSavedData()
   PrintString(execute=@n7.else, InString=FAILED TO FIND SAVED SEQUENCE)
   > Set SourceCamName()
   > Set SourceCamSeq()
  ]}
 LoopRename {nodes [
   LoopRename() {is=title K2Node_FunctionEntry}
   > n1 = GetAllActorsOfClass(ActorClass=&PanZoom2DCamera_C)
   > n2 = For Each Loop!(Exec, Array:Array of Pan Zoom 2DCamera Object References=@n1) -> (LoopBody, Array Element:Pan Zoom 2DCamera Object Reference, Array Index, Completed)
   n3 = GetDisplayName!(Object:Object Reference=@n2) -> (ReturnValue:String)
   > n4 = Branch(Condition=Contains(SearchIn=@n3, Substring=PanZoom_))
   n5 = GetDisplayName(Object=Get FocusedSequence())
   > Branch(Condition="Equal Exactly (String)"(A=@n3, B=Append(A=PanZoom_, B=@n5)))
   > HookUpPanZoom() {is=title K2Node_CallFunction}
   > Clear and Invalidate Timer by Handle(Handle=Get RenameTimer() -> (RenameTimer:Timer Handle Structure))
   n6 = Get FocusedSequence()
   SetActorLabel(execute=@n4.else, self=@n2, NewActorLabel=Append!(A=PanZoom, B=_, C:String=GetDisplayName(Object=@n6)) -> (ReturnValue), bMarkDirty=false)
   > Set 2dPanCamera(2dPanCamera=@n2)
  ]}
 OnMouseButtonUp {nodes [
   OnMouseButtonUp!() -> (then, MyGeometry:Geometry Structure, MouseEvent:"Pointer Event Structure (by ref)") {is=title K2Node_FunctionEntry}
   > Set IsDragging(IsDragging=false)
   > Return Node(ReturnValue:Event Reply Structure=Handled())
  ]}
 On_ZoomIndicator_MouseButtonDown {nodes [
   On_ZoomIndicator_MouseButtonDown!() -> (then, MyGeometry:Geometry Structure, MouseEvent:"Pointer Event Structure (by ref)") {is=title K2Node_FunctionEntry}
   n1 = SlotAsCanvasSlot(Widget=Get ZoomIndicator())
   n2 = Get 2dPanCanvas()
   > Set IndicatorStartPos(IndicatorStartPos:Vector 2D Structure="Absolute to Local (Vector)"!(
     Geometry:"Geometry Structure (by ref)"=GetCachedGeometry!(self:Widget Object Reference=@n2) -> (ReturnValue:"Geometry Structure (by ref)")
     AbsoluteVector:Vector 2D Structure=GetPosition!(self:Canvas Panel Slot Object Reference=@n1) -> (ReturnValue:Vector 2D Structure)
    ) -> (ReturnValue:Vector 2D Structure)) -> (Output_Get:Vector 2D Structure)
   > n3 = GetMousePositionOnPlatform!(execute) -> (then, ReturnValue:Vector 2D Structure)
   n4 = Get 2dPanCanvas()
   > Set MouseStartPos(MouseStartPos:Vector 2D Structure="Absolute to Local (Coordinates)"!(Geometry:"Geometry Structure (by ref)"=GetCachedGeometry(self=@n4), AbsoluteCoordinate:Vector 2D Structure=@n3) -> (ReturnValue:Vector 2D Structure)) -> (Output_Get:Vector 2D Structure)
   > Set IsDragging(IsDragging=true)
   > Return Node(ReturnValue=Handled())
  ]}
 RepositionIndicator {nodes [
   RepositionIndicator() {is=title K2Node_FunctionEntry}
   n1 = GetSize!(self=SlotAsCanvasSlot(Widget=Get 2dPanCanvas())) -> (ReturnValue_X:double, ReturnValue_Y:double)
   n2 = Get ZoomIndicator()
   n3 = Get 2dPanCanvas()
   n4 = SlotAsCanvasSlot(Widget=Get ZoomIndicator())
   n5 = vector2d + vector2d!(
    A:Vector 2D Structure="Absolute to Local (Vector)"(Geometry=GetCachedGeometry(self=@n3), AbsoluteVector=GetPosition(self=@n4))
    B:Vector 2D Structure=GetSize!(self=SlotAsCanvasSlot(Widget=@n2)) -> (ReturnValue:Vector 2D Structure)
   ) -> (ReturnValue_X:double, ReturnValue_Y:double) {is=title K2Node_PromotableOperator}
   > n6 = Branch(Condition="float > float"(A=@n5, B=@n1))
   n7 = SlotAsCanvasSlot(Widget=Get ZoomIndicator())
   n8 = GetPosition!(self=@n7) -> (ReturnValue_X:double, ReturnValue_Y:double)
   > n9 = SetPosition!(
    execute
    self:Canvas Panel Slot Object Reference=@n7
    InPosition_X:double=float - float(A=@n8, B=float - float(A=@n5, B=@n1))
    InPosition_Y:double=@n8.ReturnValue_Y
   ) -> (then)
   n10 = Branch(execute=[@n9.then, @n6.else], Condition="float > float"(A=@n5.ReturnValue_Y, B=@n1.ReturnValue_Y))
   > n11 = SetPosition(self=@n7, InPosition_X=@n8, InPosition_Y=float - float(A=@n8.ReturnValue_Y, B=float - float(A=@n5.ReturnValue_Y, B=@n1.ReturnValue_Y)))
   n12 = Get ZoomIndicator()
   n13 = CalcPositionPercentage!(InCoordinates:Vector 2D Structure=GetPosition!(self=SlotAsCanvasSlot(Widget=@n12)) -> (ReturnValue:Vector 2D Structure)) -> (CoordinatePercentage:Vector 2D Structure) {is=title K2Node_CallFunction}
   n14 = Branch(execute=[@n11.then, @n10.else], Condition="Equal (Vector2D)"(A=@n13) -> (ReturnValue=~))
   > Set PositionPercentage(PositionPercentage:Vector 2D Structure=(X=0.5,Y=0.5)) -> (Output_Get:Vector 2D Structure)
   Set PositionPercentage(execute=@n14.else, PositionPercentage=@n13)
  ]}
 ResetFilmBackTracks {nodes [
   ResetFilmBackTracks() {is=title K2Node_FunctionEntry}
   > n1 = GetBindings(Sequence=Get FocusedSequence())
   > n2 = For Each Loop with Break(Array=@n1, Break=@n3.then_0)
   > n3 = Sequence()
   n4 = GetTracks(execute=@n3.then_1, InBinding=@n2)
   > n5 = For Each Loop(Array:Array of Movie Scene Track Object References=@n4) -> (Array Element:Movie Scene Track Object Reference)
   > n6 = For Each Loop!(Exec, Array:Array of Filmback Struct Structures=Get Tracks Data()) -> (
    LoopBody
    Array Element_TrackName_2_#1:Text
    Array Element_PropertyName_5_#2:Name
    Array Element_PropertyPath_8_#3:String
    Array Index
    Completed
   )
   > n7 = GetDisplayName!(execute:Exec, Track:Movie Scene Track Object Reference=@n5) -> (then:Exec, ReturnValue:Text)
   > Branch(Condition="Equal Exactly (Text)"(A=@n7, B=@n6))
   > n8 = GetSections(Track=@n5)
   > n9 = For Each Loop!(Exec, Array:Array of Movie Scene Section Object References=@n8) -> (LoopBody, Array Element:Movie Scene Section Object Reference, Array Index, Completed)
   > n10 = GetAllChannels!(execute, Section:Movie Scene Section Object Reference=@n9) -> (then, ReturnValue:Array of Movie Scene Scripting Channel Object References)
   > n11 = For Each Loop(Array:Array of Movie Scene Scripting Channel Object References=@n10) -> (Array Element:Movie Scene Scripting Channel Object Reference)
   > n12 = Cast To MovieSceneScriptingFloatChannel!(execute, Object=@n11) -> (then, CastFailed, AsMovie Scene Scripting Float Channel:Movie Scene Scripting Float Channel Object Reference) {is=title K2Node_DynamicCast}
   > "Set Default (Float)"!(
    execute
    self:Movie Scene Scripting Float Channel Object Reference=@n12
    InDefaultValue:float="Get (a copy)"(Array=GetValues(), Dimension 1=@n6.Array Index)
   ) -> (then) {is=title K2Node_CallFunction}
  ]}
 SaveCam {nodes [
   SaveCam() {is=title K2Node_FunctionEntry}
   > n1 = Is Valid(InputObject=Get SaveGame())
   n2 = CreateSaveGameObject(execute=@n1.Is Not Valid, SaveGameClass=&SG_PanZoom_C)
   > n3 = Set SaveGame(SaveGame=@n2)
   Add!(
    execute=[@n1.Is Valid, @n3.then]
    TargetMap:Map of Level Sequence Object References to Pan Zoom Save Struct Structures=Get LvlSequences(self=Get SaveGame())
    Key:"Level Sequence Object Reference (by ref)"=Get FocusedSequence()
    Value_SourceCamName_23_#4:String=Get SourceCamName()
    Value_SourceCamSeq_25_#5:Level Sequence Object Reference=Get SourceCamSeq()
    Value_CameraFilmBack_9_#6:Camera Filmback Settings Structure=Get DefaultFilmback()
   ) -> (then)
   > SaveGameToSlot!(execute, SaveGameObject:Save Game Object Reference=Get SaveGame(), SlotName:String=PanZoom, UserIndex:int=0) -> (then, ReturnValue:bool)
  ]}
 ScaleIndicator {nodes [
   n1 = ScaleIndicator!() -> (then, Value:double=~) {is=title K2Node_FunctionEntry}
   n2 = Get ZoomIndicator()
   > Set SavedSize(SavedSize:Vector 2D Structure=GetSize(self=SlotAsCanvasSlot(Widget=@n2))) -> (Output_Get:Vector 2D Structure)
   > n3 = Set ScaleValue(ScaleValue:double=MapRangeClamped(Value=@n1, InRangeA=0, InRangeB=396, OutRangeA=1, OutRangeB=0.01)) -> (Output_Get:double)
   n4 = "To String (Float)"(InDouble=MapRangeClamped(Value=@n3, InRangeA=0.01, InRangeB=1, OutRangeA=1, OutRangeB=100))
   > Set Text(
    Text="To Text (String)"(InString=Append!(A=@n4, B=" %") -> (ReturnValue))
    self=Get ScalePercentageTxt() -> (ScalePercentageTxt:Text Object Reference)
   )
   n5 = Get!() -> (OriginalIndicatorSize_X:double, OriginalIndicatorSize_Y:double)
   > SetSize!(
    execute
    self:Canvas Panel Slot Object Reference=SlotAsCanvasSlot(Widget=Get ZoomIndicator())
    InSize_X:double=float * float(A=@n5, B=@n3)
    InSize_Y:double=float * float(A=@n5.OriginalIndicatorSize_Y, B=@n3)
   ) -> (then)
   n6 = GetPosition!(self=SlotAsCanvasSlot(Widget=Get ZoomIndicator())) -> (ReturnValue_X:double, ReturnValue_Y:double)
   n7 = GetSize!(self=SlotAsCanvasSlot(Widget=Get ZoomIndicator())) -> (ReturnValue_X:double, ReturnValue_Y:double)
   n8 = Get!() -> (SavedSize_X:double, SavedSize_Y:double)
   n9 = float * float(A=float - float(A=@n8.SavedSize_Y, B=@n7.ReturnValue_Y), B=0.5)
   n10 = float * float(A=float - float(A=@n8, B=@n7), B=0.5)
   > SetPosition(
    self=SlotAsCanvasSlot(Widget=Get ZoomIndicator())
    InPosition_X="Clamp (Float)"(Value=float + float(A=@n10, B=@n6), Min=0, Max=480)
    InPosition_Y="Clamp (Float)"(Value=float + float(A=@n9, B=@n6.ReturnValue_Y), Min=0, Max=270)
   )
   > RepositionIndicator() {is=title K2Node_CallFunction}
  ]}
 SetCamData {nodes [
   SetCamData() {is=title K2Node_FunctionEntry}
   > n1 = Is Valid(InputObject=Get ShotCineCamRef() -> (ShotCineCamRef:Cine Camera Component Object Reference))
   n2 = Is Valid(exec=@n1.Is Not Valid, InputObject=Get ShotCamRef() -> (ShotCamRef:Camera Component Object Reference))
   n3 = GetSelectedActors(execute=@n2.Is Not Valid)
   > n4 = For Each Loop(Array:Array of Actor Object References=@n3) -> (Array Element:Actor Object Reference)
   > n5 = Cast To CineCameraActor(Object=@n4)
   > n6 = Set ShotCineCamRef(ShotCineCamRef=GetCineCameraComponent(self=@n5))
   n7 = Get 2dPanCamera()
   n8 = SetData(execute=[@n6.then, @n1.Is Valid], self=@n7, Camera=Get ShotCineCamRef())
   Set DefaultFilmback(
    execute=[@n10.then, @n8.then]
    DefaultFilmback=Get InitialFilmBack!(self:Pan Zoom 2DCamera Object Reference=@n7) -> (InitialFilmBack:Camera Filmback Settings Structure)
   )
   n9 = Set ShotCamRef(execute=@n5.CastFailed, ShotCamRef=GetComponentByClass(self=@n4, ComponentClass=&CameraComponent))
   n10 = SetData(execute=[@n9.then, @n2.Is Valid], self=@n7, Camera=Get ShotCamRef())
  ]}
 SetData {nodes [
   SetData() {is=title K2Node_FunctionEntry}
   > SetCamData() {is=title K2Node_CallFunction}
   > ResetFilmBackTracks() {is=title K2Node_CallFunction}
   > Set SliderUpdated(SliderUpdated=true)
   > SetValue!(execute, self:Slider Object Reference=Get ZoomSlider() -> (ZoomSlider:Editor Utility Slider Object Reference), InValue:float=0) -> (then)
   n1 = SlotAsCanvasSlot(Widget=Get ZoomIndicator())
   > SetSize!(execute, self=@n1, InSize:Vector 2D Structure=Get OriginalIndicatorSize() -> (OriginalIndicatorSize:Vector 2D Structure)) -> (then)
   > SetPosition!(execute, self=@n1, InPosition:Vector 2D Structure) -> (then)
   > SaveCam() {is=title K2Node_CallFunction}
  ]}
 SetKey {nodes [
   n1 = SetKey!() -> (then, Track:Movie Scene Track Object Reference, NewValue:float=~) {is=title K2Node_FunctionEntry}
   > n2 = GetSections(Track=@n1)
   > n3 = For Each Loop(Array:Array of Movie Scene Section Object References=@n2) -> (Array Element:Movie Scene Section Object Reference)
   > n4 = GetAllChannels(Section=@n3)
   > n5 = For Each Loop(Array:Array of Movie Scene Scripting Channel Object References=@n4) -> (Array Element:Movie Scene Scripting Channel Object Reference)
   > n6 = Cast To MovieSceneScriptingFloatChannel(Object=@n5)
   > n7 = Get Current Local Time(TimeUnit=DisplayRate) -> (ReturnValue_Frame_SubFrame=~, ReturnValue_Time=~, ReturnValue_bHasJumped=~)
   > "Add Key (Float)"!(
    execute
    self:Movie Scene Scripting Float Channel Object Reference=@n6
    InTime_Value:int=@n7
    NewValue:float=Get NewValue() -> (NewValue:float)
    SubFrame:float=0
    TimeUnit:EMovieSceneTimeUnit Enum=DisplayRate
    InInterpolation:EMovieSceneKeyInterpolation Enum=SmartAuto
   ) -> (then, ReturnValue:Movie Scene Scripting Float Key Object Reference)
  ]}
 SetSavedData {nodes [
   n1 = SetSavedData() {is=title K2Node_FunctionEntry}
   n2 = Branch(execute=[@n1.then, @n5."  "], Condition="integer < integer"(A=Get Retry() -> (Retry:int=0), B=10))
   > n3 = FindExistingCamera() {is=title K2Node_CallFunction}
   > n4 = Branch(Condition=@n3)
   > SetData!(execute, self:Self Object Reference) -> (then) {is=title K2Node_CallFunction}
   n5 = Increment Int(" "=@n4.else, Value=Get Retry() -> (Retry=0))
   PrintString(execute=@n2.else, InString=EXPECTED CAMERA MISSING, TextColor=(R=1,G=0,B=0,A=1), Duration=10)
  ]}
 Spawn2dPanCamera {nodes [
   Spawn2dPanCamera() {is=title K2Node_FunctionEntry}
   > n1 = GetFocusedLevelSequence()
   > Set FocusedSequence(FocusedSequence=@n1)
   > n2 = GetSelectedCamInfo!(execute, self) -> (then, IsCamSelected:bool) {is=title K2Node_CallFunction}
   > Branch(Condition=@n2)
   n3 = Get FocusedSequence()
   > n4 = AddSpawnableFromClass!(
    execute
    self:Level Sequence Editor Subsystem Object Reference=Get LevelSequenceEditorSubsystem() -> (ReturnValue:Level Sequence Editor Subsystem Object Reference) {is=title K2Node_GetEditorSubsystem}
    Sequence:Movie Scene Sequence Object Reference=@n3
    ClassToSpawn:Object Class Reference=&PanZoom2DCamera_C
   ) -> (then, ReturnValue:Movie Scene Binding Proxy Structure)
   > n5 = Set PanZoomBinding(PanZoomBinding:Movie Scene Binding Proxy Structure=@n4) -> (Output_Get:Movie Scene Binding Proxy Structure)
   > SetName!(
    execute
    InBinding:"Movie Scene Binding Proxy Structure (by ref)"=@n5
    InName:String=Append!(A=PanZoom, B=_, C:String=GetDisplayName!(Object:Object Reference=@n3) -> (ReturnValue:String)) -> (ReturnValue)
   ) -> (then)
   > n6 = Set Timer by Function Name(FunctionName=LoopRename, Time=0.2, bLooping=true, bMaxOncePerFrame=false, InitialStartDelay=0, InitialStartDelayVariance=0)
   > Set RenameTimer(RenameTimer:Timer Handle Structure=@n6) -> (Output_Get:Timer Handle Structure)
  ]}
 ToggleCam {nodes [
   ToggleCam() {is=title K2Node_FunctionEntry}
   > Set CheckedState!(execute, CheckedState:ECheckBoxState Enum=Unchecked, self:Check Box Object Reference=Get AutoPilotCam()) -> (then, Output_Get:ECheckBoxState Enum=Unchecked)
   > n1 = GetPilotLevelActor(self=Get LevelEditorSubsystem(), ViewportConfigKey=None)
   > n2 = Branch(Condition="Equal (Object)"(A=@n1, B=Get 2dPanCamera()))
   n3 = Get LevelEditorSubsystem()
   n4 = Get SelectedCam()
   > PilotLevelActor(self=@n3, ActorToPilot=@n4, ViewportConfigKey=None)
   n5 = Get PilotedCamTxt() -> (PilotedCamTxt:Text Object Reference)
   > Set Text(Text="To Text (String)"(InString=GetDisplayName(Object=@n4)), self=@n5)
   n6 = Get 2dPanCamera()
   PilotLevelActor(execute=@n2.else, self=@n3, ActorToPilot=@n6, ViewportConfigKey=None)
   > Set Text(Text="To Text (String)"(InString=GetDisplayName(Object=@n6)), self=@n5)
  ]}
 UpdateIndicatorPosition {nodes [
   UpdateIndicatorPosition() {is=title K2Node_FunctionEntry}
   > n1 = GetMousePositionOnPlatform()
   n2 = Get ZoomIndicator()
   > Set OldPos(OldPos:Vector 2D Structure=GetPosition!(self=SlotAsCanvasSlot(Widget=@n2)) -> (ReturnValue:Vector 2D Structure)) -> (Output_Get:Vector 2D Structure)
   n3 = GetParent!(self:Widget Object Reference) -> (ReturnValue:Panel Widget Object Reference)
   n4 = GetCachedGeometry(self=Get 2dPanCanvas())
   n5 = vector2d + vector2d(
    A=Get IndicatorStartPos() -> (IndicatorStartPos:Vector 2D Structure)
    B=vector2d - vector2d(
     A:Vector 2D Structure="Absolute to Local (Coordinates)"(Geometry=@n4, AbsoluteCoordinate=@n1)
     B:Vector 2D Structure=Get MouseStartPos() -> (MouseStartPos:Vector 2D Structure)
    ) -> (ReturnValue:Vector 2D Structure) {is=title K2Node_PromotableOperator}
   )
   n6 = GetSize(self=SlotAsCanvasSlot(Widget=Get 2dPanCanvas()))
   n7 = GetSize(self=SlotAsCanvasSlot(Widget=Get ZoomIndicator()))
   n8 = MakeVector2D(
    X="Clamp (Float)"(Value=@n5, Min=0, Max=float - float(A=@n6, B=@n7))
    Y="Clamp (Float)"(Value=@n5.ReturnValue_Y, Min=0, Max=float - float(A=@n6.ReturnValue_Y, B=@n7.ReturnValue_Y))
   )
   > SetPosition(
    self=SlotAsCanvasSlot(Widget=Get ZoomIndicator())
    InPosition="Absolute to Local (Vector)"(Geometry=GetCachedGeometry(self=@n3), AbsoluteVector=@n8)
   )
   > n9 = Set NewPos(NewPos:Vector 2D Structure=@n8) -> (Output_Get:Vector 2D Structure)
   > Set PositionPercentage(PositionPercentage=CalcPositionPercentage(InCoordinates=@n9))
  ]}
 UpdateUI {nodes [
   UpdateUI() {is=title K2Node_FunctionEntry}
   n1 = Get ZoomIndicator()
   > Set SavedSize(SavedSize=GetSize!(self=SlotAsCanvasSlot(Widget=@n1)) -> (ReturnValue:Vector 2D Structure))
   n2 = Get!(self:Cine Camera Component Object Reference=Get 2dPanCameraComponent(self=Get 2dPanCamera())) -> (
    Filmback_SensorWidth:float=~
    Filmback_SensorHeight:float=~
    Filmback_SensorHorizontalOffset:float=~
    Filmback_SensorVerticalOffset:float=~
    Filmback_SensorAspectRatio:float=~
   )
   n3 = Get!() -> (
    DefaultFilmback_SensorWidth:float=~
    DefaultFilmback_SensorHeight:float=~
    DefaultFilmback_SensorHorizontalOffset:float=~
    DefaultFilmback_SensorVerticalOffset:float=~
    DefaultFilmback_SensorAspectRatio:float=~
   )
   n4 = float / float(A=@n2.Filmback_SensorHeight, B=@n3.DefaultFilmback_SensorHeight)
   n5 = float / float(A=@n2, B=@n3)
   n6 = MapRangeClamped(Value=float / float(A=float + float(A=@n5, B=@n4), B=2), InRangeA=0.01, InRangeB=1, OutRangeA=396, OutRangeB=0)
   > SetValue(self=Get ZoomSlider(), InValue=@n6)
   > ScaleIndicator(Value=@n6) {is=title K2Node_CallFunction}
   n7 = GetCamOffset!() -> (ReturnValue_X:double, ReturnValue_Y:double) {is=title K2Node_CallFunction}
   n8 = GetSize!(self=SlotAsCanvasSlot(Widget=Get 2dPanCanvas())) -> (ReturnValue_X:double, ReturnValue_Y:double)
   n9 = GetSize(self=SlotAsCanvasSlot(Widget=Get ZoomIndicator()))
   > SetPosition!(
    execute
    self=SlotAsCanvasSlot(Widget=Get ZoomIndicator())
    InPosition_X:double=MapRangeClamped(Value=@n7, InRangeA=-1, InRangeB=1, OutRangeB=float - float(A=@n8, B=@n9))
    InPosition_Y:double=MapRangeClamped(Value=@n7.ReturnValue_Y, InRangeA=-1, InRangeB=1, OutRangeB=float - float(A=@n8.ReturnValue_Y, B=@n9.ReturnValue_Y))
   ) -> (then)
   > Set SliderUpdated(SliderUpdated=true)
   GetSize(self=SlotAsCanvasSlot(Widget=Get 2dPanCanvas()))
  ]}
]
variables [
 2dPanCamera {type=PanZoom2DCamera_C, default=null}
 BindingProxy {type=MovieSceneBindingProxy, default {bindingId=#7, sequence=null}}
 DefaultFilmback {
  type=CameraFilmbackSettings
  default {sensorAspectRatio=1.33, sensorHeight=18.67, sensorHorizontalOffset=0, sensorVerticalOffset=0, sensorWidth=24.89}
 }
 FocusedSequence {type=LevelSequence, default=null}
 IndicatorStartPos {type=Vector2D, default=<0 0>}
 InitialFilmback {
  type=CameraFilmbackSettings
  default {sensorAspectRatio=1.33, sensorHeight=18.67, sensorHorizontalOffset=0, sensorVerticalOffset=0, sensorWidth=24.89}
 }
 IsDragging {type=boolean}
 IsFrameDifferent {type=boolean}
 MouseStartPos {type=Vector2D, default=<0 0>}
 OriginalIndicatorSize {type=Vector2D, default=<0 0>}
 PanZoomBinding {type=MovieSceneBindingProxy, default {bindingId=#7, sequence=null}}
 PositionPercentage {type=Vector2D, default=<0 0>}
 PreviousFrame {type=integer}
 RenameTimer {type=TimerHandle}
 RootSequence {type=LevelSequence, default=null}
 SaveGame {type=SG_PanZoom_C, default=null}
 ScaleValue {type=number}
 SelectedCam {type=Actor, default=null}
 SeqCheckTimer {type=TimerHandle}
 ShotCamRef {type=CameraComponent, default=null}
 ShotCineCamRef {type=CineCameraComponent, default=null}
 SliderUpdated {type=boolean}
 SourceCamName {type=string, default=""}
 SourceCamSeq {type=LevelSequence, default=null}
 Tracks Data {
  type="array<FilmbackStruct>"
  default [
   {propertyName=SensorWidth, propertyPath=Filmback.SensorWidth, trackName="Sensor Width (Filmback)"}
   {propertyName=SensorHeight, propertyPath=Filmback.SensorHeight, trackName="Sensor Height (Filmback)"}
   {propertyName=SensorHorizontalOffset, propertyPath=Filmback.SensorHorizontalOffset, trackName=Sensor Horizontal Offset}
   {propertyName=SensorVerticalOffset, propertyPath=Filmback.SensorVerticalOffset, trackName=Sensor Vertical Offset}
  ]
 }
]
widgets {
 info {rootWidgetClass=&CanvasPanel, widgetCount=21}
 widgets [
  CanvasPanel_0 {widgetClassPath=&CanvasPanel, widget=CanvasPanel_0}
  Image_0 {
   widgetClassPath=&Image
   parent=CanvasPanel_0
   widget=Image_0
   properties {brush {
     bIsDynamicallyLoaded=false
     drawAs=Image
     imageType=FullColor
     mirroring=NoMirror
     resourceName=None
     resourceObject=&/PanZoom2D_RenderTarget_Mat
     tiling=NoTile
     imageSize=<32 32>
     margin=<ltrb 0 0 0 0>
     outlineSettings {
      bUseBrushTransparency=false
      roundingType=HalfHeightRadius
      width=0
      color {colorUseRule=UseColor_Specified, specifiedColor=<rgba 0 0 0 0>}
      cornerRadii=<0 0 0 0>
     }
     tintColor {colorUseRule=UseColor_Specified, specifiedColor=<rgba 1 1 1 1>}
     uVRegion {bIsValid=false, max=<0 0>, min=<0 0>}
    }}
   slot {class=CanvasPanelSlot, properties {layoutData {alignment=<0 0>, anchors {maximum=<0.5 0>, minimum=<0.5 0>}, offsets=<ltrb -240 72 480 270>}}}
  }
  2dPanCanvas {
   widgetClassPath=&CanvasPanel
   bIsVariable=true
   parent=CanvasPanel_0
   widget=2dPanCanvas
   properties {navigation=&self:WidgetTree.2dPanCanvas.WidgetNavigation_0}
   slot {class=CanvasPanelSlot, properties {layoutData {alignment=<0 0>, anchors {maximum=<0.5 0>, minimum=<0.5 0>}, offsets=<ltrb -240 72 480 270>}}}
  }
  Image_38 {
   widgetClassPath=&Image
   parent=2dPanCanvas
   widget=Image_38
   properties {brush {
     bIsDynamicallyLoaded=false
     drawAs=Image
     imageType=FullColor
     mirroring=NoMirror
     resourceName=None
     resourceObject=&/PanZoom2D_RenderTarget
     tiling=NoTile
     imageSize=<32 32>
     margin=<ltrb 0 0 0 0>
     outlineSettings {
      bUseBrushTransparency=false
      roundingType=HalfHeightRadius
      width=0
      color {colorUseRule=UseColor_Specified, specifiedColor=<rgba 0 0 0 0>}
      cornerRadii=<0 0 0 0>
     }
     tintColor {colorUseRule=UseColor_Specified, specifiedColor=<rgba 1 1 1 1>}
     uVRegion {bIsValid=false, max=<0 0>, min=<0 0>}
    }}
   slot {class=CanvasPanelSlot, properties {layoutData {alignment=<0 0>, anchors {maximum=<0 0>, minimum=<0 0>}, offsets=<ltrb 0 0 480 270>}}}
  }
  ZoomIndicator {
   widgetClassPath=&Border
   bIsVariable=true
   parent=2dPanCanvas
   widget=ZoomIndicator
   properties {
    navigation=&self:WidgetTree.ZoomIndicator.WidgetNavigation_0
    brushColor=<rgba 1 1 1 0.25>
   }
   slot {class=CanvasPanelSlot, properties {layoutData {alignment=<0 0>, anchors {maximum=<0 0>, minimum=<0 0>}, offsets=<ltrb 0 0 480 270>}}}
  }
  AddToolBtn {
   widgetClassPath=&EditorUtilityButton
   parent=CanvasPanel_0
   widget=AddToolBtn
   properties {toolTipText=select camera you wish to pan zoom on}
   slot {
    class=CanvasPanelSlot
    properties {layoutData {alignment=<0 0>, anchors {maximum=<0.5 0>, minimum=<0.5 0>}, offsets=<ltrb -240 16 122.5225 40>}}
   }
  }
  AddToolBtnTxt {
   widgetClassPath=&TextBlock
   bIsVariable=true
   parent=AddToolBtn
   widget=AddToolBtnTxt
   properties {
    justification=Center
    text=Add 2d Pan / Reset
    font {
     bForceMonospaced=false
     bMaterialIsStencil=false
     fontMaterial=null
     fontObject=&Roboto
     letterSpacing=0
     monospacedWidth=1
     size=9
     skewAmount=0
     typefaceFontName=Bold
     outlineSettings {
      bApplyOutlineToDropShadows=false
      bMiteredCorners=false
      bSeparateFillAlpha=false
      outlineMaterial=null
      outlineSize=0
      outlineColor=<rgba 0 0 0 1>
     }
    }
   }
   slot {class=ButtonSlot}
  }
  KeyPosBtn {
   widgetClassPath=&EditorUtilityButton
   parent=CanvasPanel_0
   widget=KeyPosBtn
   slot {
    class=CanvasPanelSlot
    properties {layoutData {alignment=<0 0>, anchors {maximum=<0.5 0>, minimum=<0.5 0>}, offsets=<ltrb -224.9609 372 122.5225 40>}}
   }
  }
  AddToolBtnTxt_1 {
   widgetClassPath=&TextBlock
   parent=KeyPosBtn
   widget=AddToolBtnTxt_1
   properties {
    justification=Center
    text=Key Cam Pos
    font {
     bForceMonospaced=false
     bMaterialIsStencil=false
     fontMaterial=null
     fontObject=&Roboto
     letterSpacing=0
     monospacedWidth=1
     size=9
     skewAmount=0
     typefaceFontName=Bold
     outlineSettings {
      bApplyOutlineToDropShadows=false
      bMiteredCorners=false
      bSeparateFillAlpha=false
      outlineMaterial=null
      outlineSize=0
      outlineColor=<rgba 0 0 0 1>
     }
    }
   }
   slot {class=ButtonSlot}
  }
  ToggleCamBtn {
   widgetClassPath=&EditorUtilityButton
   parent=CanvasPanel_0
   widget=ToggleCamBtn
   slot {
    class=CanvasPanelSlot
    properties {layoutData {alignment=<0 0>, anchors {maximum=<0.5 0>, minimum=<0.5 0>}, offsets=<ltrb -60 372 122.5225 40>}}
   }
  }
  ToggleCamTxt {
   widgetClassPath=&TextBlock
   parent=ToggleCamBtn
   widget=ToggleCamTxt
   properties {
    justification=Center
    text=Toggle Pilot Cam
    font {
     bForceMonospaced=false
     bMaterialIsStencil=false
     fontMaterial=null
     fontObject=&Roboto
     letterSpacing=0
     monospacedWidth=1
     size=9
     skewAmount=0
     typefaceFontName=Bold
     outlineSettings {
      bApplyOutlineToDropShadows=false
      bMiteredCorners=false
      bSeparateFillAlpha=false
      outlineMaterial=null
      outlineSize=0
      outlineColor=<rgba 0 0 0 1>
     }
    }
   }
   slot {class=ButtonSlot}
  }
  ClearSaveBtn {
   widgetClassPath=&EditorUtilityButton
   parent=CanvasPanel_0
   widget=ClearSaveBtn
   slot {
    class=CanvasPanelSlot
    properties {layoutData {alignment=<0 0>, anchors {maximum=<0.5 0>, minimum=<0.5 0>}, offsets=<ltrb 107.0391 372 122.5225 40>}}
   }
  }
  AddToolBtnTxt_2 {
   widgetClassPath=&TextBlock
   parent=ClearSaveBtn
   widget=AddToolBtnTxt_2
   properties {
    justification=Center
    text=Clear Save Data
    font {
     bForceMonospaced=false
     bMaterialIsStencil=false
     fontMaterial=null
     fontObject=&Roboto
     letterSpacing=0
     monospacedWidth=1
     size=9
     skewAmount=0
     typefaceFontName=Bold
     outlineSettings {
      bApplyOutlineToDropShadows=false
      bMiteredCorners=false
      bSeparateFillAlpha=false
      outlineMaterial=null
      outlineSize=0
      outlineColor=<rgba 0 0 0 1>
     }
    }
   }
   slot {class=ButtonSlot}
  }
  ZoomSlider {
   widgetClassPath=/Script/Blutility.EditorUtilitySlider
   bIsVariable=true
   parent=CanvasPanel_0
   widget=ZoomSlider
   properties {maxValue=396}
   slot {
    class=CanvasPanelSlot
    properties {layoutData {alignment=<0 0>, anchors {maximum=<0.5 0>, minimum=<0.5 0>}, offsets=<ltrb -72 16 276.3763 40>}}
   }
  }
  TextBlock_1 {
   widgetClassPath=&TextBlock
   parent=CanvasPanel_0
   widget=TextBlock_1
   properties {
    text=Zoom
    font {
     bForceMonospaced=false
     bMaterialIsStencil=false
     fontMaterial=null
     fontObject=&Roboto
     letterSpacing=0
     monospacedWidth=1
     size=9
     skewAmount=0
     typefaceFontName=Bold
     outlineSettings {
      bApplyOutlineToDropShadows=false
      bMiteredCorners=false
      bSeparateFillAlpha=false
      outlineMaterial=null
      outlineSize=0
      outlineColor=<rgba 0 0 0 1>
     }
    }
   }
   slot {
    class=CanvasPanelSlot
    properties {layoutData {alignment=<0 0>, anchors {maximum=<0.5 0>, minimum=<0.5 0>}, offsets=<ltrb -112 28 38.6019 15.118>}}
   }
  }
  VerticalBox_59 {widgetClassPath=&VerticalBox, parent=CanvasPanel_0, widget=VerticalBox_59, slot {class=CanvasPanelSlot, properties=~}}
  HorizontalBox_153 {widgetClassPath=&HorizontalBox, parent=VerticalBox_59, widget=HorizontalBox_153, slot {class=VerticalBoxSlot}}
  AutoPilotCam {
   widgetClassPath=&EditorUtilityCheckBox
   bIsVariable=true
   parent=CanvasPanel_0
   widget=AutoPilotCam
   properties {checkedState=Checked}
   slot {class=CanvasPanelSlot, properties {layoutData {alignment=<0 0>, anchors {maximum=<0.5 0>, minimum=<0.5 0>}, offsets=<ltrb -48 336 100 40>}}}
  }
  TextBlock_0 {
   widgetClassPath=&TextBlock
   parent=AutoPilotCam
   widget=TextBlock_0
   properties {
    text=Auto Pilot Cam
    font {
     bForceMonospaced=false
     bMaterialIsStencil=false
     fontMaterial=null
     fontObject=&Roboto
     letterSpacing=0
     monospacedWidth=1
     size=9
     skewAmount=0
     typefaceFontName=Bold
     outlineSettings {
      bApplyOutlineToDropShadows=false
      bMiteredCorners=false
      bSeparateFillAlpha=false
      outlineMaterial=null
      outlineSize=0
      outlineColor=<rgba 0 0 0 1>
     }
    }
   }
   slot {class=PanelSlot}
  }
  ScalePercentageTxt {
   widgetClassPath=&TextBlock
   bIsVariable=true
   parent=CanvasPanel_0
   widget=ScalePercentageTxt
   properties {
    text=100
    font {
     bForceMonospaced=false
     bMaterialIsStencil=false
     fontMaterial=null
     fontObject=&Roboto
     letterSpacing=0
     monospacedWidth=1
     size=9
     skewAmount=0
     typefaceFontName=Bold
     outlineSettings {
      bApplyOutlineToDropShadows=false
      bMiteredCorners=false
      bSeparateFillAlpha=false
      outlineMaterial=null
      outlineSize=0
      outlineColor=<rgba 0 0 0 1>
     }
    }
   }
   slot {
    class=CanvasPanelSlot
    properties {layoutData {alignment=<0 0>, anchors {maximum=<0.5 0>, minimum=<0.5 0>}, offsets=<ltrb 207.0391 28 151 40>}}
   }
  }
  PilotedCamTxt {
   widgetClassPath=&TextBlock
   bIsVariable=true
   parent=CanvasPanel_0
   widget=PilotedCamTxt
   properties {
    justification=Center
    font {
     bForceMonospaced=false
     bMaterialIsStencil=false
     fontMaterial=null
     fontObject=&Roboto
     letterSpacing=0
     monospacedWidth=1
     size=9
     skewAmount=0
     typefaceFontName=Bold
     outlineSettings {
      bApplyOutlineToDropShadows=false
      bMiteredCorners=false
      bSeparateFillAlpha=false
      outlineMaterial=null
      outlineSize=0
      outlineColor=<rgba 0 0 0 1>
     }
    }
   }
   slot {
    class=CanvasPanelSlot
    properties {layoutData {alignment=<0 0>, anchors {maximum=<0.5 0>, minimum=<0.5 0>}, offsets=<ltrb -72.9609 416 151 40>}}
   }
  }
 ]
}
